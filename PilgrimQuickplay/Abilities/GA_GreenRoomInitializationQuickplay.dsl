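; Gameplay ability for the Pilgrim Quickplay green room that initializes the owning player pawn's movement state.
; On activation it removes effects tagged as Quickplay-modified movement from the owner.
; If the Pilgrim.DisableHamSandwichBackstageMovement cvar is set, non-primary local players, identified through PilgrimPlayerControllerComponent, receive a GameplayEffect that disables movement and emoting.
; The ability ends itself after a short delay.
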(event Ability|EventActivateAbility
  (bind _asfort_player_pawn (Utilities|Casting|CastToFortPlayerPawn (Ability|GetAvatarActorfromActorInfo)))
  (bind _output_get (Variables|Default|SetAs%20Fort%20Player%20Pawn _asfort_player_pawn))
  (Utilities|FlowControl|Sequence
    (:then_0
      ;[0.5722 1 0.9768 0.5] In any other scenario, remove any gameplay effects with these tags
      (Ability|RemoveGameplayEffectFromOwnerWithAssetTags "(GameplayTags=((TagName=\"Pilgrim.Quickplay.GameplayModifiedMovement\")))"))
    (:then_1
      ;[0.5722 1 0.9768 0.5] Only disable P2-P4 movement if the cvar is set
      (bind _returnvalue (Development|GetConsoleVariableBoolValue "Pilgrim.DisableHamSandwichBackstageMovement"))
      (if _returnvalue
        ;[0.5722 1 0.9768 0.5] Attempt to get a valid fort player controller and pilgrim player controller component
        (Utilities|IsValid (Fortnite|GetFortPlayerControllerfromActor _output_get)
          (:Is%20Valid
            (Utilities|IsValid (Actor|GetComponentbyClass (Fortnite|GetFortPlayerControllerfromActor _output_get) "/Script/PilgrimCoreRuntime.PilgrimPlayerControllerComponent")
              (:Is%20Valid
                ;[0.5722 1 0.9768 0.5] If this is not a primary player, apply the PilgrimQuickplayGameplay GameplayEffect to disable movement and emoting
                (if (not (Class|PilgrimPlayerControllerComponent|IsPrimaryLocalPlayer (Actor|GetComponentbyClass (Fortnite|GetFortPlayerControllerfromActor _output_get) "/Script/PilgrimCoreRuntime.PilgrimPlayerControllerComponent")))
                  (Ability|ApplyGameplayEffectToOwner "/PilgrimQuickplay/Player/Abilities/GE_PilgrimQuickplay_GreenRoom_DisableMovement.GE_PilgrimQuickplay_GreenRoom_DisableMovement_C"))))))))
    (:then_2
      ;[0.5722 1 0.9768 0.5] End the GA naturally after 2 seconds
      (Utilities|FlowControl|Delay 2.0)
      (Ability|EndAbility))))

(event Ability|EventOnEndAbility
  (params (bWasCancelled :type Boolean)))
